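# Idle Niagara effect for the Scarlet Bionic body cosmetic, putting red lightning arcs and glows on the arms and legs.
# ArmArcs and LegArcs repeatedly burst short-lived sprites at randomized intervals, oriented by shape normals around sampled skeletal bones or sockets.
# Two airglow emitters each hold two particles on those limbs, with sprites and lights flickered by waveforms.
# Partition index selects each particle's bone or socket and offset from the user-supplied arm and leg meshes.
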NiagaraSystem /BRCosmetics/Characters/Player/Female/Medium/Bodies/F_MED_ScarletBionic/FX/NS_ScarletBionic_Body_Lightning_Idle
dependencies {
 usedDataInterfaces [
  &NiagaraDataInterfaceArrayFloat3
  &NiagaraDataInterfaceCurve
  &NiagaraDataInterfaceVector2DCurve
 ]
 usedDynamicInputs []
 usedModules [&SystemState]
}
emitters {
 ArmArcs {
  emitterUpdateScript [
   &EmitterState(
    Life Cycle Mode=::NewEnumerator1
    Loop Duration=&RandomRangeFloat(Minimum=0.1, Maximum=0.4)
    Recalculate Duration Each Loop=true
    UseLoopDelay=false
   )
   &SpawnBurst_Instantaneous(
    Spawn Count=&RandomRangeInt(Minimum=2, Maximum=6)
    Use Spawn Probability=false
    UseLoopCountLimit=false
   )
  ]
  particleSpawnScript [
   &InitializeParticle(
    Color=<rgba 1 0 0 5>
    Lifetime=&RandomRangeFloat(Minimum=0.1, Maximum=0.3)
    Sprite Size=<5 5>
    Sprite Size Mode=::NewEnumerator4
    UsePositionOffset=false
    Write Material Random=true
   )
   &PartitionParticles()
   /SetVariables(Particles.ContactSpriteScale:Vector2f=&Vector2DFromFloat(Value=0.5)) {isSetParameters=true}
  ]
  particleUpdateScript [
   &ParticleState()
   &SkeletalMeshLocation(
    Bone / Socket Index=$Particles.PartitionParticles.Partition
    Bone Sampling Mode=::NewEnumerator4
    MaskSpawnGroup=false
    Orientation Sampling=::NewEnumerator1
    Position Sampling=::NewEnumerator2
    Sampled Position Offset=&SelectVectorFromArray(
     Direct Array Index=$Particles.PartitionParticles.Partition
     Vector Selection Array=DataInterface() {propertyValues {FloatData [<7 -0.5 -1>, <-7 0.5 0>]}}
    )
    Skeletal Mesh=$User.SkeletalMeshArms
   )
   &ShapeLocation(
    Additional Quaternion Rotation=true
    Additional Yaw / Pitch / Roll=&SelectVectorFromArray(
     Direct Array Index=$Particles.PartitionParticles.Partition
     Vector Selection Array=DataInterface() {propertyValues {FloatData [<0 -100 0>, <0 80 0>]}}
    )
    Disc Coverage=1
    MaskSpawnGroup=false
    Override Seed=false
    Ring Radius=0.5
    Rotation Angle Type:ENiagara_AngleInput=::NewEnumerator0
    Rotation Mode=::NewEnumerator3
    Rotation Quaternion=$Particles.SkeletalMeshLocation.SampledRotation
    Shape Origin=$Particles.SkeletalMeshLocation.SampledPosition
    Shape Primitive=::NewEnumerator4
    UseOwnerScale=true
   )
   &SpriteFacingAndAlignment(Sprite Alignment=$Particles.ShapeLocation.ShapeNormal, WriteAlignment=true, WriteFacing=false)
   &ScaleColor(
    Scale Alpha=&FloatFromCurve(FloatCurve=DataInterface() {propertyValues {Curve {keys [
         {interpMode=RCIM_Cubic, tangentMode=RCTM_User, time=0, value=0}
         {interpMode=RCIM_Cubic, tangentMode=RCTM_User, time=0.5, value=1}
         {tangentMode=RCTM_User, time=1, value=0}
        ]}}})
    ScaleA=true
    ScaleRGB=true
   )
   &ScaleSpriteSize(
    Non-Uniform Curve Sprite Scale=DataInterface() {propertyValues {
      XCurve {keys [
        {interpMode=RCIM_Cubic, tangentMode=RCTM_User, time=0, value=0}
        {arriveTangent=-0.2008, interpMode=RCIM_Cubic, leaveTangent=-0.2008, tangentMode=RCTM_User, time=0.2, value=1}
        {arriveTangent=0.2148, interpMode=RCIM_Cubic, leaveTangent=0.2148, tangentMode=RCTM_User, time=0.8, value=1}
        {tangentMode=RCTM_User, time=1, value=0}
       ]}
      YCurve {keys [{interpMode=RCIM_Cubic, tangentMode=RCTM_User, time=0, value=1}, {interpMode=RCIM_Cubic, tangentMode=RCTM_User, time=1, value=1}]}
     }}
    Scale Sprite Size Mode=::NewEnumerator3
   )
   &DynamicMaterialParameters(
    Index 0 Param 1=&RandomRangeFloat(Minimum=0.05, Maximum=0.2)
    Index 0 Param 4=1
    Module.Param0WriteEnabled=true
    Module.Param1WriteEnabled=true
    Module.Param2WriteEnabled=true
    Module.Param3WriteEnabled=true
   )
  ]
  rendererClasses [&NiagaraSpriteRendererProperties]
  renderers [
   {rendererClass=&NiagaraSpriteRendererProperties}
   {rendererClass=&NiagaraSpriteRendererProperties, rendererIndex=1}
  ]
 }
 ArmsAirglowLights {
  emitterUpdateScript [
   &EmitterState(
    Inactive Response=::NewEnumerator1
    Life Cycle Mode=::NewEnumerator1
    Loop Behavior=::NewEnumerator1
    Loop Duration Mode=::NewEnumerator1
    UseLoopDelay=false
   )
   &SpawnBurst_Instantaneous(Spawn Count=2, Use Spawn Probability=false, UseLoopCountLimit=false)
  ]
  particleSpawnScript [
   &PartitionParticles()
   &InitializeParticle(Color=<rgba 1 0 0 1>, Sprite Size Mode=::NewEnumerator3, Uniform Sprite Size=5, UsePositionOffset=false, Write Material Random=true)
   /SetVariables(Particles.LightExponent=3, Particles.LightRadius=5) {isSetParameters=true, module=SetVariables_#1}
   /SetVariables(Particles.AirGlowSpriteScale:Vector2f=&Multiply_Vector2D_ByFloat(Float=30, Vector2D=<1 1>)) {isSetParameters=true, module=SetVariables_#2}
   &CameraOffset() {enabled=false}
  ]
  particleUpdateScript [
   &ParticleState(Kill Particles When Lifetime Has Elapsed=false, Let Infinitely Lived Particles Die When Emitter Deactivates=true)
   &SkeletalMeshLocation(
    Bone / Socket Index=$Particles.PartitionParticles.Partition
    Bone Sampling Mode=::NewEnumerator4
    MaskSpawnGroup=false
    Sampled Position Offset=&SelectVectorFromArray(
     Direct Array Index=$Particles.PartitionParticles.Partition
     Vector Selection Array=DataInterface() {propertyValues {FloatData [<10 0 0>, <-10 0 0>]}}
    )
    Skeletal Mesh=$User.SkeletalMeshArms
    Use Sampled Mesh Attribute as Ribbon ID=false
   )
   /SetVariables(Particles.LightColor=&MultiplyLinearColor(
     Linear Color=<rgba 10 0 0 1>
     Scale Factor=&MakeVector4_FromFloat(Float=&Waveform(
       Global Input Position=&ReturnNormalizedExecIndex(Normalized Index Scale=20)
       Global Phase=&Add_Float(A=$Emitter.Age, B=&RandomRangeFloat(Minimum=0, Maximum=5))
       Global Amplitude Scale=50
       "Waveform [1]"=::NewEnumerator8
       "[1] Frequency"=6
       "[1] Amplitude Min/Max"=<0 1>
       "[1] Scale"=0.4
      ))
    )) {isSetParameters=true, module=SetVariables_#3}
   /SetVariables(Particles.AirGlowColor:LinearColor=&MultiplyLinearColor(
     Linear Color=<rgba 1 0 0 1>
     Scale Factor=&MakeVector4_FromFloat(Float=&Waveform(
       Global Input Position=&ReturnNormalizedExecIndex(Normalized Index Scale=20)
       Global Phase=&Add_Float(A=$Emitter.Age, B=&RandomRangeFloat(Minimum=0, Maximum=5))
       Global Amplitude Scale=0.35
       "Waveform [1]"=::NewEnumerator8
       "[1] Frequency"=6
       "[1] Amplitude Min/Max"=<0 1>
       "[1] Scale"=0.4
       "[1] Bias"=0.7
      ))
    )) {isSetParameters=true, module=SetVariables_#4}
  ]
  rendererClasses [&NiagaraSpriteRendererProperties, &NiagaraLightRendererProperties]
  renderers [
   {rendererClass=&NiagaraSpriteRendererProperties}
   {rendererClass=&NiagaraLightRendererProperties, rendererIndex=1}
   {rendererClass=&NiagaraSpriteRendererProperties, rendererIndex=2}
  ]
 }
 LegArcs {
  emitterUpdateScript [
   &EmitterState(
    Life Cycle Mode=::NewEnumerator1
    Loop Duration=&RandomRangeFloat(Minimum=0.1, Maximum=0.4)
    Recalculate Duration Each Loop=true
    UseLoopDelay=false
   )
   &SpawnBurst_Instantaneous(
    Spawn Count=&RandomRangeInt(Minimum=2, Maximum=6)
    Use Spawn Probability=false
    UseLoopCountLimit=false
   )
  ]
  particleSpawnScript [
   &InitializeParticle(
    Color=<rgba 1 0 0 2>
    Lifetime=&RandomRangeFloat(Minimum=0.1, Maximum=0.3)
    Sprite Size=<8 16>
    Sprite Size Mode=::NewEnumerator4
    UsePositionOffset=false
    Write Material Random=true
   )
   &PartitionParticles()
  ]
  particleUpdateScript [
   &ParticleState()
   &SkeletalMeshLocation(
    Bone / Socket Index=$Particles.PartitionParticles.Partition
    Bone Sampling Mode=::NewEnumerator4
    MaskSpawnGroup=false
    Orientation Sampling=::NewEnumerator1
    Position Sampling=::NewEnumerator2
    Sampled Position Offset=&SelectVectorFromArray(
     Direct Array Index=$Particles.PartitionParticles.Partition
     Vector Selection Array=DataInterface() {propertyValues {FloatData [<6 0 1>, <-6 0 -1>]}}
    )
    Skeletal Mesh=$User.SkeletalMeshLegs
   )
   &ShapeLocation(
    Additional Quaternion Rotation=true
    Additional Yaw / Pitch / Roll=&SelectVectorFromArray(
     Direct Array Index=$Particles.PartitionParticles.Partition
     Vector Selection Array=DataInterface() {propertyValues {FloatData [<0 95 0>, <0 -85 0>]}}
    )
    Disc Coverage=1
    MaskSpawnGroup=false
    Override Seed=false
    Ring Radius=1
    Rotation Angle Type:ENiagara_AngleInput=::NewEnumerator0
    Rotation Mode=::NewEnumerator3
    Rotation Quaternion=$Particles.SkeletalMeshLocation.SampledRotation
    Shape Origin=$Particles.SkeletalMeshLocation.SampledPosition
    Shape Primitive=::NewEnumerator4
    UseOwnerScale=true
   )
   /SetVariables(Particles.RandNormalAim:Vector3f=&AddVectorToPosition(
     Position=$Particles.ShapeLocation.ShapeNormal
     Vector=&RandomRangeVector(Minimum=<-0.25 -0.25 -0.25>, Maximum=<0.25 0.25 0.25>)
    )) {isSetParameters=true}
   &SpriteFacingAndAlignment(Sprite Alignment=$Particles.RandNormalAim, WriteAlignment=true, WriteFacing=false)
   &ScaleColor(
    Scale Alpha=&FloatFromCurve(FloatCurve=DataInterface() {propertyValues {Curve {keys [
         {interpMode=RCIM_Cubic, tangentMode=RCTM_User, time=0, value=0}
         {interpMode=RCIM_Cubic, tangentMode=RCTM_User, time=0.5, value=1}
         {tangentMode=RCTM_User, time=1, value=0}
        ]}}})
    ScaleA=true
    ScaleRGB=true
   )
   &ScaleSpriteSize(
    Non-Uniform Curve Sprite Scale=DataInterface() {propertyValues {
      XCurve {keys [
        {interpMode=RCIM_Cubic, tangentMode=RCTM_User, time=0, value=0}
        {arriveTangent=-0.2008, interpMode=RCIM_Cubic, leaveTangent=-0.2008, tangentMode=RCTM_User, time=0.2, value=1}
        {arriveTangent=0.2148, interpMode=RCIM_Cubic, leaveTangent=0.2148, tangentMode=RCTM_User, time=0.8, value=1}
        {tangentMode=RCTM_User, time=1, value=0}
       ]}
      YCurve {keys [{interpMode=RCIM_Cubic, tangentMode=RCTM_User, time=0, value=1}, {interpMode=RCIM_Cubic, tangentMode=RCTM_User, time=1, value=1}]}
     }}
    Scale Sprite Size Mode=::NewEnumerator3
   )
   &DynamicMaterialParameters(
    Index 0 Param 1=&RandomRangeFloat(Minimum=0.1, Maximum=0.2)
    Index 0 Param 4=1
    Module.Param0WriteEnabled=true
    Module.Param1WriteEnabled=true
    Module.Param2WriteEnabled=true
    Module.Param3WriteEnabled=true
   )
  ]
  rendererClasses [&NiagaraSpriteRendererProperties]
  renderers [
   {rendererClass=&NiagaraSpriteRendererProperties}
   {rendererClass=&NiagaraSpriteRendererProperties, rendererIndex=1}
  ]
 }
 LegsAirglowLights {
  emitterUpdateScript [
   &EmitterState(
    Inactive Response=::NewEnumerator1
    Life Cycle Mode=::NewEnumerator1
    Loop Behavior=::NewEnumerator1
    Loop Duration Mode=::NewEnumerator1
    UseLoopDelay=false
   )
   &SpawnBurst_Instantaneous(Spawn Count=2, Use Spawn Probability=false, UseLoopCountLimit=false)
  ]
  particleSpawnScript [
   &PartitionParticles()
   &InitializeParticle(Color=$System.LightColor, Sprite Size Mode=::NewEnumerator3, Uniform Sprite Size=5, UsePositionOffset=false, Write Material Random=true)
   /SetVariables(Particles.LightExponent=3, Particles.LightRadius=5) {isSetParameters=true, module=SetVariables_#1}
   /SetVariables(Particles.AirGlowSpriteScale:Vector2f=&Multiply_Vector2D_ByFloat(Float=40, Vector2D=<1 1>)) {isSetParameters=true, module=SetVariables_#2}
   &CameraOffset() {enabled=false}
  ]
  particleUpdateScript [
   &ParticleState(Kill Particles When Lifetime Has Elapsed=false, Let Infinitely Lived Particles Die When Emitter Deactivates=true)
   &SkeletalMeshLocation(
    Bone / Socket Index=$Particles.PartitionParticles.Partition
    Bone Sampling Mode=::NewEnumerator4
    MaskSpawnGroup=false
    Sampled Position Offset=&SelectVectorFromArray(
     Direct Array Index=$Particles.PartitionParticles.Partition
     Vector Selection Array=DataInterface() {propertyValues {FloatData [<-4 -2 4>, <4 2 -4>]}}
    )
    Skeletal Mesh=$User.SkeletalMeshLegs
    Use Sampled Mesh Attribute as Ribbon ID=false
   )
   /SetVariables(Particles.LightColor=&MultiplyLinearColor(
     Linear Color=<rgba 10 0 0 1>
     Scale Factor=&MakeVector4_FromFloat(Float=&Waveform(
       Global Input Position=&ReturnNormalizedExecIndex(Normalized Index Scale=20)
       Global Phase=&Add_Float(A=$Emitter.Age, B=&RandomRangeFloat(Minimum=0, Maximum=5))
       Global Amplitude Scale=100
       "Waveform [1]"=::NewEnumerator8
       "[1] Frequency"=6
       "[1] Amplitude Min/Max"=<0 1>
       "[1] Scale"=0.4
      ))
    )) {isSetParameters=true, module=SetVariables_#3}
   /SetVariables(Particles.AirGlowColor:LinearColor=&MultiplyLinearColor(
     Linear Color=<rgba 1 0 0 1>
     Scale Factor=&MakeVector4_FromFloat(Float=&Waveform(
       Global Input Position=&ReturnNormalizedExecIndex(Normalized Index Scale=20)
       Global Phase=&Add_Float(A=$Emitter.Age, B=&RandomRangeFloat(Minimum=0, Maximum=5))
       Global Amplitude Scale=0.35
       "Waveform [1]"=::NewEnumerator8
       "[1] Frequency"=6
       "[1] Amplitude Min/Max"=<0 1>
       "[1] Scale"=0.4
       "[1] Bias"=0.7
      ))
    )) {isSetParameters=true, module=SetVariables_#4}
  ]
  rendererClasses [&NiagaraSpriteRendererProperties, &NiagaraLightRendererProperties]
  renderers [
   {rendererClass=&NiagaraSpriteRendererProperties}
   {rendererClass=&NiagaraLightRendererProperties, rendererIndex=1}
   {rendererClass=&NiagaraSpriteRendererProperties, rendererIndex=2}
  ]
 }
}
summary {
 systemName=NS_ScarletBionic_Body_Lightning_Idle
 emitters [
  {emitterName=ArmArcs, rendererClasses [&NiagaraSpriteRendererProperties]}
  {emitterName=LegArcs, rendererClasses [&NiagaraSpriteRendererProperties]}
  {
   emitterName=ArmsAirglowLights
   rendererClasses [&NiagaraSpriteRendererProperties, &NiagaraLightRendererProperties]
  }
  {
   emitterName=LegsAirglowLights
   rendererClasses [&NiagaraSpriteRendererProperties, &NiagaraLightRendererProperties]
  }
 ]
 userVariables [
  User.SkeletalMeshArms {
   type=NiagaraDataInterfaceSkeletalMesh
   defaultValue {
    struct=NiagaraExt_VariableValue_DataInterface
    value {dataInterface=&self:NiagaraDataInterfaceSkeletalMesh_0}
   }
  }
  User.SkeletalMeshLegs {
   type=NiagaraDataInterfaceSkeletalMesh
   defaultValue {
    struct=NiagaraExt_VariableValue_DataInterface
    value {dataInterface=&self:NiagaraDataInterfaceSkeletalMesh_1}
   }
  }
 ]
}
