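# Physics asset for the ScarletBionic female medium face accessory mesh, defining dynamic hair bodies and body collision shapes.
# A seven-capsule back hair chain hangs from the head, alongside single front, left and right hair capsules, all with tight angular limits.
# Head, spine, upper arm and thigh shapes give a coarse body volume, linked by free angular constraints.
# A very large sphere on root, offset downward, anchors the lower spine and thigh constraints.
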PhysicsAsset /BRCosmetics/Characters/Player/Female/Medium/Bodies/F_MED_ScarletBionic/Meshes/Parts/F_MED_ScarletBionic_FaceAcc_Physics
bodies {
 dyn_hair_b [{shapeType=Capsule, length=10, radius=7, center=<5.7499 0 0>, rotation=<rot 90 0 0>}]
 dyn_hair_b_1 [{shapeType=Capsule, length=10, radius=5, center=<5.7499 0 0>, rotation=<rot 90 0 0>}]
 dyn_hair_b_2 [{shapeType=Capsule, length=10, radius=5, center=<5.7499 0 0>, rotation=<rot 90 0 0>}]
 dyn_hair_b_3 [{shapeType=Capsule, length=10, radius=5, center=<5.7499 0 0>, rotation=<rot 90 0 0>}]
 dyn_hair_b_4 [{shapeType=Capsule, length=10, radius=5, center=<5.7499 0 0>, rotation=<rot 90 0 0>}]
 dyn_hair_b_5 [{shapeType=Capsule, length=10, radius=5, center=<5.7499 0 0>, rotation=<rot 90 0 0>}]
 dyn_hair_b_6 [{shapeType=Capsule, length=10, radius=5, center=<5.7499 0 0>, rotation=<rot 90 0 0>}]
 dyn_hair_f [{shapeType=Capsule, length=5, radius=5, center=<3.1649 0 0>}]
 dyn_hair_l [{
   shapeType=Capsule
   length=3.2779
   radius=2.2232
   shapeName=dyn_hair_l_capsule
   center=<2.8905 -0.7722 0.4558>
   rotation=<rot 0 65.9996 -76.6322>
  }]
 dyn_hair_r [{
   shapeType=Capsule
   length=3.2139
   radius=2.3221
   shapeName=dyn_hair_r_capsule
   center=<-2.793 0.8188 -0.3705>
   rotation=<rot 0 72.3507 -80.5772>
  }]
 head [{shapeType=Sphere, radius=9.0779, shapeName=head_sphere, center=<6.1914 0 0>}]
 root [{shapeType=Sphere, radius=165.5076, shapeName=root_sphere, center=<0 0 -147.5311>}]
 spine_02 [{shapeType=Capsule, length=10.0347, radius=11, shapeName=spine_02_capsule, center=<3.7936 3.6039 0>, rotation=<rot 0 -90 0>}]
 spine_05 [
  {shapeType=Capsule, length=10.0347, radius=12, shapeName=spine_05_capsule, center=<2.3521 3.6039 0>, rotation=<rot 0 -90 0>}
  {shapeType=Capsule, length=15, radius=5, shapeName=spine_05_capsule, center=<17.0662 0.888 0>}
 ]
 thigh_l [{shapeType=Capsule, length=20, radius=13, center=<-10.6095 0 3.3666>, rotation=<rot 90 0 0>}]
 thigh_r [{shapeType=Capsule, length=20, radius=13, center=<10.6095 0 -3.3666>, rotation=<rot 90 0 0>}]
 upperarm_l [{shapeType=Capsule, length=20, radius=6, shapeName=upperarm_l_capsule, center=<12.6116 0 0>, rotation=<rot 90 0 0>}]
 upperarm_r [{shapeType=Capsule, length=20, radius=6, shapeName=upperarm_r_capsule, center=<-12.6116 0 0>, rotation=<rot 90 0 0>}]
}
constraints [
 {
  bone1Name=head
  bone2Name=spine_05
  swing1LimitDegrees=45
  swing1Motion=Free
  swing2LimitDegrees=45
  swing2Motion=Free
  twistLimitDegrees=45
  twistMotion=Free
 }
 {bone1Name=dyn_hair_b, bone2Name=head, swing1LimitDegrees=30, swing2LimitDegrees=20, twistLimitDegrees=5}
 {bone1Name=dyn_hair_b_1, bone2Name=dyn_hair_b, swing1LimitDegrees=30, swing2LimitDegrees=25, twistLimitDegrees=5}
 {bone1Name=dyn_hair_b_2, bone2Name=dyn_hair_b_1, swing1LimitDegrees=30, swing2LimitDegrees=25, twistLimitDegrees=5}
 {bone1Name=dyn_hair_b_3, bone2Name=dyn_hair_b_2, swing1LimitDegrees=30, swing2LimitDegrees=25, twistLimitDegrees=5}
 {bone1Name=dyn_hair_b_4, bone2Name=dyn_hair_b_3, swing1LimitDegrees=30, swing2LimitDegrees=25, twistLimitDegrees=5}
 {bone1Name=dyn_hair_b_5, bone2Name=dyn_hair_b_4, swing1LimitDegrees=30, swing2LimitDegrees=25, twistLimitDegrees=5}
 {bone1Name=dyn_hair_b_6, bone2Name=dyn_hair_b_5, swing1LimitDegrees=30, swing2LimitDegrees=25, twistLimitDegrees=5}
 {bone1Name=dyn_hair_f, bone2Name=head, swing1LimitDegrees=30, swing2LimitDegrees=30, twistLimitDegrees=10}
 {bone1Name=dyn_hair_l, bone2Name=head, swing1LimitDegrees=25, swing2LimitDegrees=25, twistLimitDegrees=5}
 {bone1Name=dyn_hair_r, bone2Name=head, swing1LimitDegrees=25, swing2LimitDegrees=25, twistLimitDegrees=5}
 {
  bone1Name=spine_05
  bone2Name=spine_02
  swing1LimitDegrees=45
  swing1Motion=Free
  swing2LimitDegrees=45
  swing2Motion=Free
  twistLimitDegrees=45
  twistMotion=Free
 }
 {
  bone1Name=upperarm_r
  bone2Name=spine_05
  swing1LimitDegrees=45
  swing1Motion=Free
  swing2LimitDegrees=45
  swing2Motion=Free
  twistLimitDegrees=45
  twistMotion=Free
 }
 {
  bone1Name=upperarm_l
  bone2Name=spine_05
  swing1LimitDegrees=45
  swing1Motion=Free
  swing2LimitDegrees=45
  swing2Motion=Free
  twistLimitDegrees=45
  twistMotion=Free
 }
 {
  bone1Name=spine_02
  bone2Name=root
  swing1LimitDegrees=45
  swing1Motion=Free
  swing2LimitDegrees=45
  swing2Motion=Free
  twistLimitDegrees=45
  twistMotion=Free
 }
 {
  bone1Name=thigh_l
  bone2Name=root
  swing1LimitDegrees=45
  swing1Motion=Free
  swing2LimitDegrees=45
  swing2Motion=Free
  twistLimitDegrees=45
  twistMotion=Free
 }
 {
  bone1Name=thigh_r
  bone2Name=root
  swing1LimitDegrees=45
  swing1Motion=Free
  swing2LimitDegrees=45
  swing2Motion=Free
  twistLimitDegrees=45
  twistMotion=Free
 }
]
