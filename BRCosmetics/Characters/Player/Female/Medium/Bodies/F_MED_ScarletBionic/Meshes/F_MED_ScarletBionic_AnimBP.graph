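# Post-process style AnimBlueprint for the F_MED_ScarletBionic female body, adding secondary motion to its hair bones.
# It copies the parent mesh pose through a Control Rig, with constrained, head-pitch-corrected back hair for non-cape backbling or DBNO carrying.
# Trail, AnimDynamics and RigidBody nodes simulate the hair, weighted by backbling type and carry state.
# The EventGraph sets those weights, gravity override and cape-only backward force.
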AnimBlueprint /BRCosmetics/Characters/Player/Female/Medium/Bodies/F_MED_ScarletBionic/Meshes/F_MED_ScarletBionic_AnimBP
target_skeleton=&/F_MED_ScarletBionic_Skeleton
graphs [
 AnimGraph {nodes [
   n1 = "Transform (Modify) Bone - Bone: dyn_hair_b"(
    ComponentPose=Local To Component(LocalPose=Use cached pose 'Constrained'())
    Alpha=1
    Translation="0,0,0"
    Rotation="-0.016,6.537,0.026"
    Scale="1,1,1"
   ) {
    rotationMode=BMM_Replace
    rotationSpace=BCS_BoneSpace
    translationMode=BMM_Replace
    translationSpace=BCS_BoneSpace
    alphaBoolBlend {blendOption=Linear}
    boneToModify {boneName=dyn_hair_b}
    rotation=<rot -0.016 6.537 0.026>
   }
   n2 = "Transform (Modify) Bone - Bone: dyn_hair_b_2"(ComponentPose=@n1, Alpha=1, Translation="0.195,1.877,-0.090", Rotation="0.021,-4.902,0.043", Scale="1,1,1") {
    rotationMode=BMM_Replace
    rotationSpace=BCS_BoneSpace
    translationMode=BMM_Replace
    translationSpace=BCS_BoneSpace
    alphaBoolBlend {blendOption=Linear}
    boneToModify {boneName=dyn_hair_b_2}
    rotation=<rot 0.021 -4.902 0.043>
    translation=<0.195 1.877 -0.09>
   }
   n3 = Component To Local(ComponentPose=@n2)
   n4 = "Transform (Modify) Bone - Bone: dyn_hair_b"(
    ComponentPose=Local To Component(LocalPose=Use cached pose 'Constrained'())
    Alpha=1
    Translation="0,0,0"
    Rotation="0.031,7.401,0.014"
    Scale="1,1,1"
   ) {
    rotationMode=BMM_Replace
    rotationSpace=BCS_BoneSpace
    translationMode=BMM_Replace
    translationSpace=BCS_BoneSpace
    alphaBoolBlend {blendOption=Linear}
    boneToModify {boneName=dyn_hair_b}
    rotation=<rot 0.031 7.401 0.014>
   }
   n5 = "Transform (Modify) Bone - Bone: dyn_hair_b_2"(ComponentPose=@n4, Alpha=1, Translation="0.195,1.877,-0.090", Rotation="0.016,0.530,-0.037", Scale="1,1,1") {
    rotationMode=BMM_Replace
    rotationSpace=BCS_BoneSpace
    translationMode=BMM_Replace
    translationSpace=BCS_BoneSpace
    alphaBoolBlend {blendOption=Linear}
    boneToModify {boneName=dyn_hair_b_2}
    rotation=<rot 0.0161 0.53 -0.037>
    translation=<0.195 1.877 -0.09>
   }
   n6 = Component To Local(ComponentPose=@n5)
   n7 = Get head_fwdBwd_rot() -> (head_fwdBwd_rot:double)
   n8 = "Transform (Modify) Bone - Bone: dyn_hair_b"(
    ComponentPose=Local To Component(LocalPose=Blend Poses by bool(BlendPose_0=@n3, BlendPose_1=@n6, BlendTime_0=0.5, BlendTime_1=0.5) {blendType=Linear})
    Alpha=@n7
    Translation="0.401,-2.802,-0.000"
    Rotation="0.000,-10.939,-0.000"
    Scale="1,1,1"
   ) {
    rotationMode=BMM_Additive
    rotationSpace=BCS_BoneSpace
    translationMode=BMM_Additive
    translationSpace=BCS_BoneSpace
    alphaBoolBlend {blendOption=Linear}
    alphaScaleBiasClamp {bMapRange=true, inRange {max=0, min=-25}, outRange {max=0, min=1}}
    boneToModify {boneName=dyn_hair_b}
    rotation {yaw=-10.939}
    translation=<0.401 -2.802>
   }
   n9 = "Transform (Modify) Bone - Bone: dyn_hair_b_1"(ComponentPose=@n8, Alpha=@n7, Translation="0.863,4.712,-0.193", Rotation="0.005,5.554,0.002", Scale="1,1,1") {
    rotationMode=BMM_Additive
    rotationSpace=BCS_BoneSpace
    translationMode=BMM_Additive
    translationSpace=BCS_BoneSpace
    alphaBoolBlend {blendOption=Linear}
    alphaScaleBiasClamp {bMapRange=true, inRange {max=0, min=-25}, outRange {max=0, min=1}}
    boneToModify {boneName=dyn_hair_b_1}
    rotation=<rot 0.005 5.554 0.002>
    translation=<0.863 4.712 -0.193>
   }
   n10 = Get head_fwdBwd_rot()
   n11 = "Transform (Modify) Bone - Bone: dyn_hair_b"(ComponentPose=@n9, Alpha=@n10, Translation="0,1,0", Rotation="0,0,0", Scale="1,1,1") {
    rotationMode=BMM_Additive
    rotationSpace=BCS_BoneSpace
    translationMode=BMM_Additive
    translationSpace=BCS_BoneSpace
    alphaBoolBlend {blendOption=Linear}
    alphaScaleBiasClamp {bMapRange=true, inRange {max=15}}
    boneToModify {boneName=dyn_hair_b}
    translation {y=1}
   }
   n12 = "Transform (Modify) Bone - Bone: dyn_hair_b_1"(ComponentPose=@n11, Alpha=@n10, Translation="0.000,-1.177,1.259", Rotation="-0.036,-1.705,-0.046", Scale="1,1,1") {
    rotationMode=BMM_Additive
    rotationSpace=BCS_BoneSpace
    translationMode=BMM_Additive
    translationSpace=BCS_BoneSpace
    alphaBoolBlend {blendOption=Linear}
    alphaScaleBiasClamp {bMapRange=true, inRange {max=15}}
    boneToModify {boneName=dyn_hair_b_1}
    rotation=<rot -0.036 -1.705 -0.046>
    translation {y=-1.177, z=1.259}
   }
   n13 = Blend Poses by bool(
    BlendPose_0=Component To Local(ComponentPose=@n12)
    BlendPose_1=Use cached pose 'BodyPose'()
    BlendTime_0=0.5
    BlendTime_1=0.5
    bActiveValue=Get ActivateBackpackSetup() -> (ActivateBackpackSetup:bool)
   ) {blendType=Linear}
   n14 = Use cached pose 'BodyPose'()
   n15 = Get Is Backbling A Cape() -> (Is Backbling A Cape:bool)
   n16 = Get BackPackOrDBNOOverride() -> (BackPackOrDBNOOverride:double)
   n17 = Trail controller - Bone: dyn_hair_b_6!(
    ComponentPose=Local To Component(LocalPose=Blend Poses by bool(BlendPose_0=@n14, BlendPose_1=@n13, BlendTime_0=0.5, BlendTime_1=0.5, bActiveValue=@n15) {blendType=Linear})
    Alpha=@n16
    RelaxationSpeedScale:float=@n16
   ) -> (Pose) {
    bLimitRotation=true
    chainLength=6
    maxDeltaTime=0.0333
    alphaScaleBiasClamp {bInterpResult=true}
    rotationLimits [
     {limitMax=<0 20 0>, limitMin=<0 -20 0>}
     {limitMax=<0 20 0>, limitMin=<0 -20 0>}
     {limitMax=<10 180 10>, limitMin=<-10 -180 -10>}
     {limitMax=<10 180 10>, limitMin=<-10 -180 -10>}
     {limitMax=<10 180 10>, limitMin=<-10 -180 -10>}
     {limitMax=<10 180 10>, limitMin=<-10 -180 -10>}
    ]
    rotationOffsets [<0 0 0>, <0 0 0>, <0 0 0>, <0 0 0>, <0 0 0>, <0 0 0>]
    trailBone {boneName=dyn_hair_b_6}
    trailRelaxationSpeed {editorCurveData {keys [<at 0 10 RCIM_Linear>,<at 0.4971 9.4102 RCIM_Linear>,<at 1.0035 8.5176 RCIM_Linear>]}}
   }
   n18 = Anim Dynamics - Bone: dyn_hair_l(ComponentPose=@n17, Alpha=1, ExternalForce="0,0,0") {
    angularDampingOverride=1
    angularSpringConstant=10000
    bAngularSpring=true
    bGravityOverrideInSimSpace=true
    bOverrideAngularDamping=true
    bUseGravityOverride=true
    gravityScale=0.35
    simulationSpace=RootRelative
    alphaBoolBlend {blendOption=Linear}
    boundBone {boneName=dyn_hair_l}
    componentLinearAccScale=<1 1 1>
    componentLinearVelScale=<1 1 1>
    gravityOverride {z=-980}
    physicsBodyDefinitions [{
      sphereCollisionRadius=0
      boundBone {boneName=dyn_hair_l}
      boxExtents=<10 20 20>
      constraintSetup {angularLimitsMax=<5 5 5>, angularLimitsMin=<-15 0 -15>, angularTarget=<1 0 0>}
      localJointOffset=<0.1 0 0>
     }]
   }
   n19 = Anim Dynamics - Bone: dyn_hair_r(ComponentPose=@n18, Alpha=1, ExternalForce="0,0,0") {
    angularDampingOverride=1
    angularSpringConstant=10000
    bAngularSpring=true
    bGravityOverrideInSimSpace=true
    bOverrideAngularDamping=true
    bUseGravityOverride=true
    gravityScale=0.35
    simulationSpace=RootRelative
    alphaBoolBlend {blendOption=Linear}
    boundBone {boneName=dyn_hair_r}
    componentLinearAccScale=<1 1 1>
    componentLinearVelScale=<1 1 1>
    gravityOverride {z=-980}
    physicsBodyDefinitions [{
      sphereCollisionRadius=0
      boundBone {boneName=dyn_hair_r}
      boxExtents=<10 20 20>
      constraintSetup {angularLimitsMax=<5 5 5>, angularLimitsMin=<-15 0 -15>, angularTarget=<1 0 0>}
      localJointOffset=<-0.1 0 0>
     }]
   }
   n20 = Anim Dynamics - Bone: dyn_hair_f(ComponentPose=@n19, Alpha=1, ExternalForce="0,0,0") {
    angularDampingOverride=1
    angularSpringConstant=10000
    bAngularSpring=true
    bGravityOverrideInSimSpace=true
    bOverrideAngularDamping=true
    bUseGravityOverride=true
    gravityScale=0.35
    simulationSpace=RootRelative
    alphaBoolBlend {blendOption=Linear}
    boundBone {boneName=dyn_hair_f}
    componentAppliedLinearAccClamp=<2000 2000 2000>
    componentLinearAccScale=<1 1 1>
    componentLinearVelScale=<1 1 1>
    gravityOverride {z=-980}
    physicsBodyDefinitions [{
      sphereCollisionRadius=0
      boundBone {boneName=dyn_hair_f}
      boxExtents=<10 20 20>
      constraintSetup {angularLimitsMax=<5 5 5>, angularLimitsMin=<-15 -15 -15>, angularTarget=<1 0 0>}
      localJointOffset=<1 0 0>
     }]
   }
   n21 = RigidBody!(
    ComponentPose=@n20
    Alpha=Get BackPackOrDBNOOverride()
    OverrideWorldGravity:Vector=Get gravity_override() -> (gravity_override:Vector)
    ExternalForce=Get external force() -> (external force:Vector)
   ) -> (Pose) {
    bClampLinearTranslationLimitToRefPose=true
    bOverrideWorldGravity=true
    overridePhysicsAsset=/BRCosmetics/Characters/Player/Female/Medium/Bodies/F_MED_ScarletBionic/Meshes/Parts/F_MED_ScarletBionic_FaceAcc_Physics
    simulationSpace=BaseBoneSpace
    alphaBoolBlend {blendOption=Linear}
    baseBoneRef {boneName=spine_02}
    componentAppliedLinearAccClamp=<3000 3000 3000>
    componentLinearAccScale=<0.4 0.35 0.8>
    componentLinearVelScale=<0.4 0.35 0.8>
    overrideWorldGravity {z=-980}
    simSpaceSettings {maxAngularAcceleration=0, maxAngularVelocity=2000, maxLinearAcceleration=0, maxLinearVelocity=2000, worldAlpha=1}
   }
   n22 = Get ShortBackPackOrDBNOOverride() -> (ShortBackPackOrDBNOOverride:double)
   n23 = Anim Dynamics - Bone: dyn_hair_b_3(ComponentPose=@n21, Alpha=@n22, ExternalForce="0,0,0") {
    angularDampingOverride=1
    angularSpringConstant=20000
    bAngularSpring=true
    bGravityOverrideInSimSpace=true
    bOverrideAngularDamping=true
    bOverrideLinearDamping=true
    bUseGravityOverride=true
    gravityScale=0.35
    linearDampingOverride=1
    simulationSpace=RootRelative
    alphaBoolBlend {blendOption=Linear}
    alphaScaleBiasClamp {bMapRange=true, outRange {max=0.5}}
    boundBone {boneName=dyn_hair_b_3}
    componentAppliedLinearAccClamp=<3000 3000 3000>
    componentLinearAccScale=<0.5 1 1>
    componentLinearVelScale=<0.5 1 1>
    gravityOverride {z=-980}
    physicsBodyDefinitions [{
      sphereCollisionRadius=0
      boundBone {boneName=dyn_hair_b_3}
      boxExtents=<13 40 40>
      constraintSetup {angularLimitsMax=<5 45 10>, angularLimitsMin=<-5 -45 -45>, angularTarget=<1 0 0>}
      localJointOffset=<10 0 0>
     }]
    relativeSpaceBone {boneName=pelvis}
   }
   n24 = Anim Dynamics - Bone: dyn_hair_b_4(ComponentPose=@n23, Alpha=@n22, ExternalForce="0,0,0") {
    angularDampingOverride=1
    angularSpringConstant=20000
    bAngularSpring=true
    bGravityOverrideInSimSpace=true
    bOverrideAngularDamping=true
    bOverrideLinearDamping=true
    bUseGravityOverride=true
    gravityScale=0.35
    linearDampingOverride=1
    simulationSpace=RootRelative
    alphaBoolBlend {blendOption=Linear}
    alphaScaleBiasClamp {bMapRange=true, outRange {max=0.5}}
    boundBone {boneName=dyn_hair_b_4}
    componentAppliedLinearAccClamp=<3000 3000 3000>
    componentLinearAccScale=<0.5 1 1>
    componentLinearVelScale=<0.5 1 1>
    gravityOverride {z=-980}
    physicsBodyDefinitions [{
      sphereCollisionRadius=0
      boundBone {boneName=dyn_hair_b_4}
      boxExtents=<13 40 40>
      constraintSetup {angularLimitsMax=<5 45 10>, angularLimitsMin=<-5 -45 -45>, angularTarget=<1 0 0>}
      localJointOffset=<10 0 0>
     }]
    relativeSpaceBone {boneName=pelvis}
   }
   n25 = Anim Dynamics - Bone: dyn_hair_b_5(ComponentPose=@n24, Alpha=@n22, ExternalForce="0,0,0") {
    angularDampingOverride=1
    angularSpringConstant=20000
    bAngularSpring=true
    bGravityOverrideInSimSpace=true
    bOverrideAngularDamping=true
    bOverrideLinearDamping=true
    bUseGravityOverride=true
    gravityScale=0.35
    linearDampingOverride=1
    simulationSpace=RootRelative
    alphaBoolBlend {blendOption=Linear}
    alphaScaleBiasClamp {bMapRange=true, outRange {max=0.5}}
    boundBone {boneName=dyn_hair_b_5}
    componentAppliedLinearAccClamp=<3000 3000 3000>
    componentLinearAccScale=<0.5 1 1>
    componentLinearVelScale=<0.5 1 1>
    gravityOverride {z=-980}
    physicsBodyDefinitions [{
      sphereCollisionRadius=0
      boundBone {boneName=dyn_hair_b_5}
      boxExtents=<13 40 40>
      constraintSetup {angularLimitsMax=<5 45 10>, angularLimitsMin=<-5 -45 -45>, angularTarget=<1 0 0>}
      localJointOffset=<10 0 0>
     }]
    relativeSpaceBone {boneName=pelvis}
   }
   n26 = Anim Dynamics - Bone: dyn_hair_b_6(ComponentPose=@n25, Alpha=@n22, ExternalForce="0,0,0") {
    angularDampingOverride=1
    angularSpringConstant=20000
    bAngularSpring=true
    bGravityOverrideInSimSpace=true
    bOverrideAngularDamping=true
    bOverrideLinearDamping=true
    bUseGravityOverride=true
    gravityScale=0.35
    linearDampingOverride=1
    simulationSpace=RootRelative
    alphaBoolBlend {blendOption=Linear}
    alphaScaleBiasClamp {bMapRange=true, outRange {max=0.5}}
    boundBone {boneName=dyn_hair_b_6}
    componentAppliedLinearAccClamp=<3000 3000 3000>
    componentLinearAccScale=<0.5 1 1>
    componentLinearVelScale=<0.5 1 1>
    gravityOverride {z=-980}
    physicsBodyDefinitions [{
      sphereCollisionRadius=0
      boundBone {boneName=dyn_hair_b_6}
      boxExtents=<13 40 40>
      constraintSetup {angularLimitsMax=<5 45 10>, angularLimitsMin=<-5 -45 -45>, angularTarget=<0 0 0>}
      localJointOffset=<10 0 0>
     }]
    relativeSpaceBone {boneName=pelvis}
   }
   n27 = Get Is Backbling A Cape()
   n28 = "Transform (Modify) Bone - Bone: dyn_hair_b"!(ComponentPose=@n26, bAlphaBoolEnabled:bool=@n27, Translation="-0.056,0,0.056", Rotation="-0.000,-9.261,-0.003", Scale="1,1,1") -> (Pose) {
    alphaInputType=Bool
    rotationMode=BMM_Replace
    rotationSpace=BCS_BoneSpace
    translationMode=BMM_Replace
    translationSpace=BCS_BoneSpace
    alphaBoolBlend {blendOption=Linear}
    boneToModify {boneName=dyn_hair_b}
    rotation=<rot 0 -9.261 -0.003>
    translation {x=-0.056, z=0.056}
   }
   n29 = "Transform (Modify) Bone - Bone: dyn_hair_b_1"!(ComponentPose=@n28, bAlphaBoolEnabled:bool=@n27, Translation="0.001,0,0", Rotation="0.000,2.730,0.016", Scale="1,1,1") -> (Pose) {
    alphaInputType=Bool
    rotationMode=BMM_Replace
    rotationSpace=BCS_BoneSpace
    translationMode=BMM_Replace
    translationSpace=BCS_BoneSpace
    alphaBoolBlend {blendOption=Linear}
    boneToModify {boneName=dyn_hair_b_1}
    rotation=<rot 0 2.7301 0.016>
    translation {x=0.001}
   }
   Output Pose(Result=Component To Local(ComponentPose=@n29)) {layerGroup=None}
   n30 = Control Rig(Source=Copy Pose From Mesh() {bUseAttachedParent=true}) {
    alphaBoolBlend {blendOption=Linear}
    controlRigAssetReference {blueprintRigClass=&/F_MED_ScarletBionic_CtrlRig.F_MED_ScarletBionic_CtrlRig_C}
   }
   Save cached pose 'BodyPose'(Pose=@n30) {cacheName=BodyPose}
   n31 = "Constraint - dyn_hair_b (PP)"!(ComponentPose=Local To Component(LocalPose=Use cached pose 'BodyPose'()), Alpha=1, ConstraintWeights_0=0.7, ConstraintWeights_1:float=0.3) -> (Pose) {
    alphaBoolBlend {blendOption=Linear}
    boneToModify {boneName=dyn_hair_b}
    constraintSetup [{targetBone {boneName=head}}, {targetBone {boneName=spine_05}}]
    constraintWeights [1,1]
   }
   n32 = "Constraint - dyn_hair_b_1 (PP)"!(ComponentPose=@n31, Alpha=1, ConstraintWeights_0=0.3, ConstraintWeights_1:float=0.7) -> (Pose) {
    alphaBoolBlend {blendOption=Linear}
    boneToModify {boneName=dyn_hair_b_1}
    constraintSetup [{targetBone {boneName=head}}, {targetBone {boneName=spine_05}}]
    constraintWeights [1,1]
   }
   n33 = "Constraint - dyn_hair_b_2 (PPP)"!(ComponentPose=@n32, Alpha=1, ConstraintWeights_0=0.2, ConstraintWeights_1:float=0.2, ConstraintWeights_2:float=0.6) -> (Pose) {
    alphaBoolBlend {blendOption=Linear}
    boneToModify {boneName=dyn_hair_b_2}
    constraintSetup [{targetBone {boneName=pelvis}}, {targetBone {boneName=spine_04}}, {targetBone {boneName=spine_05}}]
    constraintWeights [1,1,1]
   }
   n34 = "Constraint - dyn_hair_b_3 (PPP)"!(ComponentPose=@n33, Alpha=1, ConstraintWeights_0=0.7, ConstraintWeights_1:float=0.2, ConstraintWeights_2:float=0.1) -> (Pose) {
    alphaBoolBlend {blendOption=Linear}
    boneToModify {boneName=dyn_hair_b_3}
    constraintSetup [{targetBone {boneName=pelvis}}, {targetBone {boneName=spine_04}}, {targetBone {boneName=spine_05}}]
    constraintWeights [1,1,1]
   }
   n35 = "Constraint - dyn_hair_b_4 (PP)"!(ComponentPose=@n34, Alpha=0.5, ConstraintWeights_0=0.8, ConstraintWeights_1:float=0.2) -> (Pose) {
    alphaBoolBlend {blendOption=Linear}
    boneToModify {boneName=dyn_hair_b_4}
    constraintSetup [{targetBone {boneName=pelvis}}, {targetBone {boneName=spine_02}}]
    constraintWeights [1,1]
   }
   n36 = "Constraint - dyn_hair_b_5 (PP)"!(ComponentPose=@n35, Alpha=0.2, ConstraintWeights_0=0.9, ConstraintWeights_1:float=0.1) -> (Pose) {
    alphaBoolBlend {blendOption=Linear}
    boneToModify {boneName=dyn_hair_b_5}
    constraintSetup [{targetBone {boneName=pelvis}}, {targetBone {boneName=spine_02}}]
    constraintWeights [1,1]
   }
   Save cached pose 'Constrained'(Pose=Component To Local(ComponentPose=@n36)) {cacheName=Constrained}
   Get gravity_override()
  ]}
 EventGraph {nodes [
   n1 = PlayerMovement(AnimInstance=Self-Reference(), XDotProductCoefficient=100, YDotProductCoefficient=100, ZDotProductCoefficient=100)
   n2 = MakeVector(X=@n1, Y=@n1.YMovement)
   > Set Forward(Forward:bool="float > float"(A=BreakVector(InVec=@n2).Y, B=40)) -> (Output_Get:bool)
   Event BlueprintUpdateAnimation()
   > n3 = Sequence!(execute) -> (then_0, then_1, then_2:Exec, then_3:Exec)
   > n4 = Gravity Override State Function DEPRECATED(CharacterPartAnimInstance=Self-Reference(), GameTimeInSeconds=GetGameTimeInSeconds(), GravityOverrideParams=Get Gravity Override Params())
   > Set gravity_override(gravity_override=@n4)
   n5 = PlayerMovement(execute=@n3.then_2, AnimInstance=Self-Reference(), XDotProductCoefficient=1, YDotProductCoefficient=1, ZDotProductCoefficient=1)
   n6 = Get PawnSpeed2D()
   n7 = Get Is Backbling A Cape()
   > n8 = Branch(Condition=AND Boolean!(A=@n7, B="float > float"(A=@n5.YMovement, B=0.1), C:bool="float > float"(A=@n6, B=250)) -> (ReturnValue))
   n9 = vector * vector(A=GetActorForwardVector!(self:Actor Object Reference=Get PawnOwner() -> (PawnOwner:Pawn Object Reference)) -> (ReturnValue:Vector), B=-1) -> (ReturnValue=~)
   > Set external force(external force:Vector=vector * vector(A=@n9, B:float=6000) -> (ReturnValue=~)) -> (Output_Get:Vector)
   n10 = Branch(execute=@n8.else, Condition=AND Boolean!(A=@n7, B=Get bIsInFrontEnd()) -> (ReturnValue))
   > Set external force(external force=vector * vector(A=@n9, B=2000) -> (ReturnValue=~))
   Set external force(execute=@n10.else, external force="0, 0, 0")
   n11 = Branch(
    execute=@n3.then_1
    Condition="Nearly Equal (Float)"(A=Get IsShortBackblingOnAlpha() -> (IsShortBackblingOnAlpha:double), B=1, ErrorTolerance=0.1)
   )
   n12 = GetDeltaTransformFromRefPose(self=GetOwnerSkeletalMeshComponent(), BoneName=head, BaseName=neck_01)
   > Set head_fwdBwd_rot(head_fwdBwd_rot:double=BreakRotator(InRot=BreakTransform(InTransform=@n12).Rotation).Yaw) -> (Output_Get:double)
   Set head_fwdBwd_rot(execute=@n11.else, head_fwdBwd_rot=0)
   Set ActivateBackpackSetup(
    execute=@n3.then_3
    ActivateBackpackSetup:bool=OR Boolean(A=Get IsWearingBackbling() -> (IsWearingBackbling:bool), B=Get bIsDBNOCarrying() -> (bIsDBNOCarrying:bool))
   ) -> (Output_Get:bool)
   > n13 = Sequence!(execute) -> (then_0, then_1)
   n14 = Branch(execute=@n13.then_1, Condition=Get bIsDBNOCarrying())
   > Set ShortBackPackOrDBNOOverride(ShortBackPackOrDBNOOverride:double=1) -> (Output_Get:double)
   Set ShortBackPackOrDBNOOverride(execute=@n14.else, ShortBackPackOrDBNOOverride=Get IsShortBackblingOnAlpha())
   n15 = Branch(execute=@n13.then_0, Condition=Get bIsDBNOCarrying())
   > Set BackPackOrDBNOOverride(BackPackOrDBNOOverride:double=0) -> (Output_Get:double)
   Set BackPackOrDBNOOverride(execute=@n15.else, BackPackOrDBNOOverride=Get NotBackblingOrBackblingIsCape() -> (NotBackblingOrBackblingIsCape:double))
   Event OnCharacterCustomizationLoaded()
   n16 = CheckForBackbling()
   > n17 = Set IsWearingBackbling(IsWearingBackbling:bool=@n16) -> (Output_Get:bool)
   > Set Is Backbling A Cape(Is Backbling A Cape:bool=@n16.bIsBackblingACape) -> (Output_Get:bool)
   n18 = NOT Boolean(A=@n16)
   > Set NotBackblingOrBackblingIsCape(NotBackblingOrBackblingIsCape:double="To Float (Boolean)"(InBool=OR Boolean(A=@n18, B=@n16.bIsBackblingACape))) -> (Output_Get:double)
   > n19 = Branch(Condition=@n17)
   > n20 = Branch(Condition=@n16.bIsBackblingLong)
   Set IsShortBackblingOnAlpha(execute=[@n20.then, @n19.else], IsShortBackblingOnAlpha:double=0) -> (Output_Get:double)
   Set IsShortBackblingOnAlpha(execute=@n20.else, IsShortBackblingOnAlpha=1)
   Event BlueprintBeginPlay() {is=title K2Node_Event}
   > Set PawnOwner(PawnOwner:Pawn Object Reference=TryGetPawnOwner()) -> (Output_Get:Pawn Object Reference)
  ]}
]
variables [
 ActivateBackpackSetup {type=boolean}
 BackPackOrDBNOOverride {type=number}
 external force {type=Vector}
 Forward {type=boolean}
 Gravity Override Params {type=GravityOverrideParamsStruct}
 gravity_override {type=Vector}
 head_fwdBwd_rot {type=number}
 Is Backbling A Cape {type=boolean}
 IsShortBackblingOnAlpha {type=number}
 IsWearingBackbling {type=boolean}
 NotBackblingOrBackblingIsCape {type=number}
 PawnOwner {type=Pawn, default=null}
 ShortBackPackOrDBNOOverride {type=number}
]
